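# Sound cue for the Paprika auto pistol's drop impact, playing randomized landing sounds in the loot-collect sound class.
# A quality-level node sends its first two slots to a random pick of four waves and its last two to a pick of only two.
# A root modulator varies the result.
# Attenuation and concurrency overrides keep it short-range and block new instances beyond a small simultaneous limit.
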SoundCue /PaprikaCoreWeapons/Items/Weapons/PaprikaPistol/Sounds/Pistol_Paprika_AutoPistol_DropImpact_Cue
root=Modulator_0
nodes [
 {id=Modulator_0, children [QualityLevel_0]}
 {id=QualityLevel_0, children [Random_1,Random_1,Random_0,Random_0]}
 {id=Random_0, children [WavePlayer_5,WavePlayer_6], properties {bShouldExcludeFromBranchCulling=true, weights [1,1]}}
 {id=Random_1, children [WavePlayer_3,WavePlayer_4,WavePlayer_5,WavePlayer_6], properties {weights [1,1,1,1]}}
 {id=WavePlayer_3, properties {soundWaveAssetPtr=&/Pistol_Paprika_AutoPistol_DropLand_01}}
 {id=WavePlayer_4, properties {soundWaveAssetPtr=&/Pistol_Paprika_AutoPistol_DropLand_02}}
 {id=WavePlayer_5, properties {soundWaveAssetPtr=&/Pistol_Paprika_AutoPistol_DropLand_03}}
 {id=WavePlayer_6, properties {soundWaveAssetPtr=&/Pistol_Paprika_AutoPistol_DropLand_04}}
]
properties {
 bOverrideAttenuation=true
 bOverrideConcurrency=true
 maxDistance=1000
 soundClassObject=&Fort_Loot_Collect
 volumeMultiplier=0.6
 attenuationOverrides {falloffDistance=1000, attenuationShapeExtents {x=0}}
 concurrencyOverrides {maxCount=3, resolutionRule=PreventNew}
}
