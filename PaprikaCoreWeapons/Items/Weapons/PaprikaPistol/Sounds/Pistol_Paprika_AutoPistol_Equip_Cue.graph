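# Sound cue for the Paprika auto pistol's equip sound, classed as weapon foley and using the weapon reload attenuation.
# A modulator feeds from a quality-level switch whose first two slots randomly pick among four equip waves.
# The last two slots share a smaller random node limited to waves 03 and 04, which is excluded from branch culling.
# The cue overrides concurrency and has debug enabled.
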SoundCue /PaprikaCoreWeapons/Items/Weapons/PaprikaPistol/Sounds/Pistol_Paprika_AutoPistol_Equip_Cue
root=Modulator_0
nodes [
 {id=Modulator_0, children [QualityLevel_0]}
 {id=QualityLevel_0, children [Random_0,Random_0,Random_1,Random_1]}
 {id=Random_0, children [WavePlayer_14, WavePlayer_15, WavePlayer_16, WavePlayer_17], properties {weights [1,1,1,1]}}
 {id=Random_1, children [WavePlayer_16, WavePlayer_17], properties {bShouldExcludeFromBranchCulling=true, weights [1,1]}}
 {id=WavePlayer_14, properties {soundWaveAssetPtr=&/Pistol_Paprika_AutoPistol_Equip_01}}
 {id=WavePlayer_15, properties {soundWaveAssetPtr=&/Pistol_Paprika_AutoPistol_Equip_02}}
 {id=WavePlayer_16, properties {soundWaveAssetPtr=&/Pistol_Paprika_AutoPistol_Equip_03}}
 {id=WavePlayer_17, properties {soundWaveAssetPtr=&/Pistol_Paprika_AutoPistol_Equip_04}}
]
properties {
 attenuationSettings=&Fort_Weapon_Reload_Att
 bDebug=true
 bOverrideConcurrency=true
 maxDistance=2250
 soundClassObject=&Fort_WeaponFoley
 volumeMultiplier=0.301
}
