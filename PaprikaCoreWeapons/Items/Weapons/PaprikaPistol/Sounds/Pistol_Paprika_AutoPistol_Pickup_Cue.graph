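# Sound cue for picking up the Paprika auto pistol, classed as loot collection audio.
# A modulator feeds a quality-level switch: the first two slots randomize evenly across four pickup waves, the last two across a two-wave subset.
# The two-wave random node is excluded from branch culling.
# Playback has reduced volume and pitch, uses shared weapon-pickup attenuation, and references the PickupMaximum_Oldest concurrency group.
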SoundCue /PaprikaCoreWeapons/Items/Weapons/PaprikaPistol/Sounds/Pistol_Paprika_AutoPistol_Pickup_Cue
root=Modulator_1
nodes [
 {id=Modulator_1, children [QualityLevel_0]}
 {id=QualityLevel_0, children [Random_0,Random_0,Random_1,Random_1]}
 {id=Random_0, children [WavePlayer_10, WavePlayer_11, WavePlayer_12, WavePlayer_13], properties {weights [1,1,1,1]}}
 {id=Random_1, children [WavePlayer_12, WavePlayer_13], properties {bShouldExcludeFromBranchCulling=true, weights [1,1]}}
 {id=WavePlayer_10, properties {soundWaveAssetPtr=&/Pistol_Paprika_AutoPistol_PickUp_04}}
 {id=WavePlayer_11, properties {soundWaveAssetPtr=&/Pistol_Paprika_AutoPistol_PickUp_03}}
 {id=WavePlayer_12, properties {soundWaveAssetPtr=&/Pistol_Paprika_AutoPistol_PickUp_02}}
 {id=WavePlayer_13, properties {soundWaveAssetPtr=&/Pistol_Paprika_AutoPistol_PickUp_01}}
]
properties {
 attenuationSettings=&Fort_Weapon_Pickup_Att
 maxDistance=3500
 pitchMultiplier=0.9
 soundClassObject=&Fort_Loot_Collect
 volumeMultiplier=0.6
 concurrencyOverrides {maxCount=1, resolutionRule=PreventNew}
 concurrencySet [&PickupMaximum_Oldest]
}
