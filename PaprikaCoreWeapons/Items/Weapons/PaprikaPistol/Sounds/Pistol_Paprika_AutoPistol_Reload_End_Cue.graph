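# Sound cue for the end of the Paprika auto pistol's reload, playing as weapon foley with Fortnite's reload attenuation.
# A Modulator feeds a quality-level switch that picks between two equally weighted Random nodes.
# The first two quality slots draw from six reload-end waves, while the last two use only waves 01 to 03.
# The three-wave branch is excluded from branch culling, and the cue overrides concurrency and has its debug flag on.
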SoundCue /PaprikaCoreWeapons/Items/Weapons/PaprikaPistol/Sounds/Pistol_Paprika_AutoPistol_Reload_End_Cue
root=Modulator_0
nodes [
 {id=Modulator_0, children [QualityLevel_0]}
 {id=QualityLevel_0, children [Random_1,Random_1,Random_0,Random_0]}
 {id=Random_0, children [WavePlayer_5,WavePlayer_6,WavePlayer_7], properties {bShouldExcludeFromBranchCulling=true, weights [1,1,1]}}
 {id=Random_1, children [WavePlayer_2,WavePlayer_3,WavePlayer_4,WavePlayer_5,WavePlayer_6,WavePlayer_7], properties {weights [1,1,1,1,1,1]}}
 {
  id=WavePlayer_2
  properties {soundWaveAssetPtr=&/Pistol_Paprika_AutoPistol_Reload_End_06}
 }
 {
  id=WavePlayer_3
  properties {soundWaveAssetPtr=&/Pistol_Paprika_AutoPistol_Reload_End_05}
 }
 {
  id=WavePlayer_4
  properties {soundWaveAssetPtr=&/Pistol_Paprika_AutoPistol_Reload_End_04}
 }
 {
  id=WavePlayer_5
  properties {soundWaveAssetPtr=&/Pistol_Paprika_AutoPistol_Reload_End_03}
 }
 {
  id=WavePlayer_6
  properties {soundWaveAssetPtr=&/Pistol_Paprika_AutoPistol_Reload_End_02}
 }
 {
  id=WavePlayer_7
  properties {soundWaveAssetPtr=&/Pistol_Paprika_AutoPistol_Reload_End_01}
 }
]
properties {
 attenuationSettings=&Fort_Weapon_Reload_Att
 bDebug=true
 bOverrideConcurrency=true
 maxDistance=2250
 soundClassObject=&Fort_WeaponFoley
 volumeMultiplier=0.601
}
