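# Sound cue for the start of the Paprika auto pistol's reload, played as weapon foley with Fortnite reload attenuation.
# A quality-level switch picks a random node: the first two slots draw evenly from six wave variants, the last two from only variants 04 to 06.
# A root modulator at default settings varies the chosen wave.
# The three-variant node is excluded from branch culling, and the cue overrides concurrency and has debug enabled.
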SoundCue /PaprikaCoreWeapons/Items/Weapons/PaprikaPistol/Sounds/Pistol_Paprika_AutoPistol_Reload_Start_Cue
root=Modulator_0
nodes [
 {id=Modulator_0, children [QualityLevel_0]}
 {id=QualityLevel_0, children [Random_0,Random_0,Random_1,Random_1]}
 {id=Random_0, children [WavePlayer_0,WavePlayer_1,WavePlayer_2,WavePlayer_3,WavePlayer_4,WavePlayer_5], properties {weights [1,1,1,1,1,1]}}
 {id=Random_1, children [WavePlayer_3,WavePlayer_4,WavePlayer_5], properties {bShouldExcludeFromBranchCulling=true, weights [1,1,1]}}
 {
  id=WavePlayer_0
  properties {soundWaveAssetPtr=&/Pistol_Paprika_AutoPistol_Reload_Start_01}
 }
 {
  id=WavePlayer_1
  properties {soundWaveAssetPtr=&/Pistol_Paprika_AutoPistol_Reload_Start_02}
 }
 {
  id=WavePlayer_2
  properties {soundWaveAssetPtr=&/Pistol_Paprika_AutoPistol_Reload_Start_03}
 }
 {
  id=WavePlayer_3
  properties {soundWaveAssetPtr=&/Pistol_Paprika_AutoPistol_Reload_Start_04}
 }
 {
  id=WavePlayer_4
  properties {soundWaveAssetPtr=&/Pistol_Paprika_AutoPistol_Reload_Start_05}
 }
 {
  id=WavePlayer_5
  properties {soundWaveAssetPtr=&/Pistol_Paprika_AutoPistol_Reload_Start_06}
 }
]
properties {
 attenuationSettings=&Fort_Weapon_Reload_Att
 bDebug=true
 bOverrideConcurrency=true
 maxDistance=2250
 soundClassObject=&Fort_WeaponFoley
 volumeMultiplier=1.002
}
